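# UI-domain material for an arrow button background fade in the Discovery Browser, using Modulate blending to darken what is behind it.
# A horizontal UV gradient, rotated in full turns by Rotation and shaped by EaseOutQuint, multiplies a mirrored vertical edge mask softened by SmoothStep up to Range.
# Vertex alpha and Opacity scale the mask, which is then inverted.
# A separate remap-and-ease branch on the vertical mask is left unused.
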Material /DiscoveryBrowser/Art/Materials/Buttons/M_UI_Button_Arrow_BgFade
graph {
 properties {
  bAutomaticallySetUsageInEditor=false
  bUsedWithStaticMesh=false
  blendMode=BLEND_Modulate
  materialDomain=MD_UI
  usedShadingModels=Default Lit
 }
 nodes [
  n1 = MaterialFunctionCall() {materialFunction=&GetUserInterfaceUV}
  n2 = ComponentMask(None=@n1.Normalized UV) {g=true}
  n3 = Multiply(A=Min(A=OneMinus(None=@n2), B=@n2)) {constB=2}
  MaterialFunctionCall("Value (0-1)"=MaterialFunctionCall(Input=@n3, Input Low=Constant(), Input High=Constant() {r=0.4}) {materialFunction=&RemapValue_0-1}.Result) {materialFunction=&MF_UI_EaseCurves}
  n4 = Multiply(A=VertexColor().A, B=ScalarParameter() {defaultValue=1, parameterName=Opacity, sliderMax=1}) {constB=0.5}
  n5 = ScalarParameter() {parameterName=Rotation}
  n6 = MaterialFunctionCall("Value (0-1)"=ComponentMask(None=Rotator(Coordinate=@n1.Normalized UV, Time=@n5) {speed=6.2832}) {r=true}) {materialFunction=&MF_UI_EaseCurves}
  n7 = SmoothStep(Max=ScalarParameter() {defaultValue=0.45, parameterName=Range, sliderMax=1}, Value=@n3) {constMax=0.45}
  OneMinus(None=Multiply(A=Multiply(A=@n4, B=@n6.EaseOutQuint), B=Saturate(None=@n7)))
 ]
}
